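; Gameplay ability for a small creature in the Idler set, named as a heal on death.
; On activation it commits, then runs a target selection task with application tag Targeting.ApplicationTag.0.
; When targets arrive, it applies the matching gameplay effect container to them and ends the ability.
; The healing effect and death trigger are not defined in this graph, and the end-ability event is empty.
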(event Ability|EventActivateAbility
  (Ability|CommitAbility)
  (bind (_asynctaskproxy _targetdata _applicationtag) (Ability|Tasks|DoTargetSelection "None" "(TagName=\"Targeting.ApplicationTag.0\")" true)
    (:Targeted
      (Ability|ApplyGameplayEffectContainer _targetdata _applicationtag)
      (Ability|EndAbility))))

(event Ability|EventOnEndAbility
  (params (bWasCancelled :type Boolean)))
